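# Niagara effect for the ReptilianOcean emote that adds pink-red glowing snake-eye flares, a tail glow and a flickering light to the character's skeletal mesh.
# User SecondBeat and ThirdBeat times drive burst spawns, blink flashes and fade curves, with socket sampling placing the sprites.
# Flares stack several sprite renderers with per-layer size and color written to module namespaces, tinted by time of day.
# The RoundShape emitter is disabled.
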NiagaraSystem /BRCosmetics/Animation/Game/MainPlayer/Emotes/ReptilianOcean/FX/NS_Emote_ReptilianOcean_SnakeEyes
dependencies {
 usedDataInterfaces [&NiagaraDataInterfaceVector2DCurve, &NiagaraDataInterfaceCurve]
 usedDynamicInputs []
 usedModules [
  &SystemState
  &self:SetFootDustFXColorArrayIndex
 ]
}
emitters {
 Default_LightAndGlowFlicker_Loop {
  emitterUpdateScript [
   &EmitterState(
    Inactive Response=::NewEnumerator1
    Life Cycle Mode=::NewEnumerator1
    Loop Behavior=::NewEnumerator1
    Loop Duration Mode=::NewEnumerator1
    UseLoopDelay=false
   )
   &SpawnBurst_Instantaneous(Use Spawn Probability=false, UseLoopCountLimit=false)
   /Niagara/Noise/Waveform(
    EnableExponent1=false
    GlobalClampMax=false
    GlobalClampMin=false
    UseInput1=false
    UsePhase1=false
    "Waveform [1]":ENiagara_Waveforms=::NewEnumerator12
    "[1] Frequency"=7
   )
   /SetVariables(Emitter.Flicker:NiagaraFloat=&Add_Float(A=&Multiply_Float(A=$Emitter.Waveform.Waveform, B=0.5), B=0.5)) {isSetParameters=true}
  ]
  particleSpawnScript [
   &InitializeParticle(
    Color=<rgba 1 0.0439 0.4263 0.7>
    Mesh Scale Mode=::NewEnumerator3
    Mesh Uniform Scale=10
    Position Offset Coordinate Space=::Local
    Sprite Size Mode=::NewEnumerator3
    Uniform Sprite Size=0
    UsePositionOffset=true
    Write Material Random=true
   )
   &DynamicMaterialParameters(
    Index 0 Param 1=0.2
    Index 0 Param 2=1
    Module.Param0WriteEnabled=true
    Module.Param1WriteEnabled=true
    Module.Param2WriteEnabled=true
    Module.Param3WriteEnabled=true
   )
  ]
  particleUpdateScript [
   &ParticleState(Kill Particles When Lifetime Has Elapsed=false, Let Infinitely Lived Particles Die When Emitter Deactivates=true)
   &ScaleColor(Scale Alpha=&Multiply_Float(A=$Emitter.Flicker, B=0.3), ScaleA=true, ScaleRGB=true)
   &FN_Light_Attributes(
    Color=$Particles.Color
    Create Color Attribute=true
    Create Exponent Attribute=true
    Create Radius Attribute=true
    Exponent=5
    Radius=100
    Scale Color=&Multiply_Float(
     A=&FloatFromCurve(
      FloatCurve=DataInterface() {propertyValues {Curve {keys [<at -1.2 0>,<at -1 1>,<at 0 1.0155>,<at 0.3 1>,<at 1 0.5>]}}}
      CurveIndex=&Subtract_Float(A=$Emitter.Age, B=&Add_Float(A=$User.SecondBeat, B=0))
      Scale Curve=&Multiply_Float(A=50, B=&FNTimeOfDayScalar(Night Scale=0.4, Morning Scale=0.4, Evening Scale=0.6))
     )
     B=$Emitter.Flicker
    )
   )
   &SkeletalMeshLocation(Bone Sampling Mode=::NewEnumerator4, MaskSpawnGroup=false, Sampled Position Offset=<-40 20 0>, Skeletal Mesh=$User.Skeletal Mesh)
  ]
  rendererClasses [
   &NiagaraSpriteRendererProperties
   &NiagaraLightRendererProperties
   &NiagaraMeshRendererProperties
  ]
  renderers [
   {rendererClass=&NiagaraSpriteRendererProperties}
   {rendererClass=&NiagaraLightRendererProperties, rendererIndex=1}
   {rendererClass=&NiagaraMeshRendererProperties, rendererIndex=2}
  ]
 }
 FLARE {
  emitterUpdateScript [
   &EmitterState()
   &SpawnBurst_Instantaneous(
    Spawn Count=2
    Spawn Probability=HlslExpression() {hlslExpression=1.0}
    Spawn Time=$User.SecondBeat
    Use Spawn Probability=false
    UseLoopCountLimit=true
   )
  ]
  particleSpawnScript [
   &ApplyInitialForces()
   &InitializeParticle(
    Color=&MultiplyLinearColor(
     Linear Color=&FNTimeOfDayColor(Night Color=<rgba 20 6 12.0607 1>, Morning Color=<rgba 5 0 1.3634 1>, Day Color=<rgba 5 0 1.3634 1>, Evening Color=<rgba 5 0 1.3634 1>)
     Scale Factor=<0.02 0.02 0.02 1>
    )
    Sprite Size Mode=::NewEnumerator3
    UsePositionOffset=false
    Write Material Random=true
   )
   &DynamicMaterialParameters(
    Index 0 Param 1=0.5
    Module.Index0Param0Write=true
    Module.Index0Param1Write=true
    Module.Index0Param2Write=true
    Module.Index0Param3Write=true
   )
   /SetVariables(Particles.CameraOffsetDark=5, Particles.DY0:Vector4f=<1 0 0 0>) {isSetParameters=true, module=SetVariables_#1}
   &AlignSpriteToMeshOrientation(
    Mesh Orientation Relative Sprite Alignment Vector=<0 0 1>
    Mesh Orientation Relative Sprite Facing Vector:Vector3f=<0 -1 0>
    SetAlignment=true
    SetFacing=true
   ) {enabled=false}
  ]
  particleUpdateScript [
   &ParticleState(Kill Particles When Lifetime Has Elapsed=false, Let Infinitely Lived Particles Die When Emitter Deactivates=true)
   &SkeletalMeshLocation(
    Bone / Socket Index=&ReturnExecIndex()
    MaskSpawnGroup=false
    Mesh Sampling Type=::NewEnumerator5
    Sampled Position Offset=<-2 0 0>
    Skeletal Mesh=$User.Skeletal Mesh
    Socket Sampling Mode=::NewEnumerator5
   )
   &FN_Sprite_Attributes(
    Color=&FNTimeOfDayColor(
     Night Color=<rgba 0.4 0 0.0775 0.6>
     Morning Color=<rgba 0.4 0 0.0775 0.5>
     Day Color=<rgba 0.4 0.0612 0.2415 0.1>
     Evening Color=<rgba 0.4 0 0.0775 0.3>
    )
    Create Color Attribute=true
    Create Position Attribute=true
    Create SpriteSize Attribute=true
    Position:NiagaraPosition=&AddVectorToPosition(Position=$Particles.Position, Vector=&TransformVector(Vector=<0 0 0>, Destination Space=::World))
    SpriteSize=&Vector2DFromFloat(Value=20)
    Write to Module Namespace=true
   )
   &FN_Sprite_Attributes(
    Color=&MultiplyLinearColor(
     Linear Color=$Particles.Color
     Scale Factor=&MakeVector4_FromLinearColor(LinearColor=<rgba 3 2.94 2.94 0.03>)
    )
    Create Color Attribute=true
    Create SpriteSize Attribute=true
    SpriteSize=&Multiply_Vector2D(A=$Particles.SpriteSize, B=&Vector2DFromFloat(Value=2.5))
    Write to Module Namespace=true
   ) {module=FN_Sprite_Attributes001}
   &FN_Sprite_Attributes(
    Color=&MultiplyLinearColor(
     Linear Color=$Particles.Color
     Scale Factor=&MakeVector4_FromLinearColor(LinearColor=&MultiplyLinearColor(
       Linear Color=<rgba 3 2.94 2.94 0.05>
       Scale Factor=&MakeVector4(
        X=1
        Y=1
        Z=1
        W=&Multiply_Float(
         A=&Add_Float(
          A=&Sine(
           Normalized Angle=&ClampFloat(Float=&Subtract_Float(A=$Emitter.Age, B=$User.SecondBeat), Max=40000000)
           Period=1
           Scale=0.2
           Bias=0
          )
          B=&FloatFromCurve(
           FloatCurve=DataInterface() {propertyValues {Curve {keys [<at 0.5 1>,<at 0.9 0>]}}}
           CurveIndex=&Subtract_Float(A=$Emitter.Age, B=$User.ThirdBeat)
          )
         )
         B=1
        )
       )
      ))
    )
    Create Color Attribute=true
    Create SpriteSize Attribute=true
    SpriteSize=&Multiply_Vector2D(
     A=&Vector2DFromCurve(Vector2Curve=DataInterface() {propertyValues {XCurve {keys [<at 0 5>,<at 1 2>]}, YCurve {keys [<at 0 3>,<at 0.5 1>]}}}, ScaleCurve=<1 1>)
     B=&Vector2DFromFloat(Value=100)
    )
    Write to Module Namespace=true
   ) {module=FN_Sprite_Attributes002}
   &ScaleColor(
    Scale Mode=::NewEnumerator1
    Scale RGBA=&MakeVector4_FromLinearColor(LinearColor=&FNTimeOfDayColor(Night Color=<rgba 4 4 4 4>, Morning Color=<rgba 1 1 1 4>, Day Color=<rgba 3 3 3 3>, Evening Color=<rgba 1 1 1 4>))
    ScaleRGBA=true
   ) {module=ScaleColor002}
  ]
  rendererClasses [&NiagaraSpriteRendererProperties]
  renderers [
   {rendererClass=&NiagaraSpriteRendererProperties}
   {rendererClass=&NiagaraSpriteRendererProperties, rendererIndex=1}
   {rendererClass=&NiagaraSpriteRendererProperties, rendererIndex=2}
   {rendererClass=&NiagaraSpriteRendererProperties, rendererIndex=3}
  ]
 }
 RoundShape {
  bEnabled=false
  emitterUpdateScript [
   &EmitterState()
   &SpawnRate(
    SpawnRate=&FloatFromCurve(
     FloatCurve=DataInterface() {propertyValues {Curve {keys [<at -0.01 0 RCIM_Constant>,<at 0 1 RCIM_Constant>,<at 4 0 RCIM_Constant>]}}}
     CurveIndex=&Subtract_Float(A=$Emitter.Age, B=&Add_Float(A=$User.SecondBeat, B=1))
    )
    Use Spawn Probability=false
   )
  ]
  particleSpawnScript [
   &InitializeParticle(
    Color=<rgba 3 1.5 2.2869 0.25>
    Lifetime=0.8
    Mass Mode=::NewEnumerator3
    Position Offset=$User.Offset
    Sprite Size Mode=::NewEnumerator3
    Uniform Sprite Size=80
    UsePositionOffset=true
    Write Material Random=true
   )
   &ShapeLocation(MaskSpawnGroup=false, Non Uniform Scale=<0.5 1 1>, Override Seed=false, Sphere Radius=0, UseOwnerScale=true)
   &CameraOffset(Camera Offset Amount=4)
  ]
  particleUpdateScript [
   &ParticleState()
   &SkeletalMeshLocation(
    Bone / Socket Index=2
    MaskSpawnGroup=false
    Mesh Sampling Type=::NewEnumerator5
    Skeletal Mesh=$User.Skeletal Mesh
    Socket Sampling Mode=::NewEnumerator5
   )
   &SpriteRotationRate(Rotation Rate=&FloatFromCurve(
     FloatCurve=DataInterface() {propertyValues {Curve {keys [
         {arriveTangent=1.4306, interpMode=RCIM_Cubic, leaveTangent=1.4306, tangentMode=RCTM_User, time=0.2, value=0}
         {arriveTangent=1.1179, interpMode=RCIM_Cubic, leaveTangent=1.1179, tangentMode=RCTM_User, time=0.4, value=0.6228}
         {arriveTangent=0.0224, interpMode=RCIM_Cubic, leaveTangent=0.0224, tangentMode=RCTM_User, time=1, value=1}
        ]}}}
     Scale Curve=250
    )) {enabled=false}
   &ScaleColor(
    Scale Alpha=&FloatFromCurve(FloatCurve=DataInterface() {propertyValues {Curve {keys [
         {arriveTangent=9.0725, interpMode=RCIM_Cubic, leaveTangent=9.0725, tangentMode=RCTM_User, time=0, value=0.4}
         {arriveTangent=-2.5249, interpMode=RCIM_Cubic, leaveTangent=-2.5249, tangentMode=RCTM_User, time=0.1408, value=0.9847}
         <at 0.6 1>
         {arriveTangent=-0.0405, interpMode=RCIM_Cubic, leaveTangent=-0.0405, tangentMode=RCTM_User, time=1, value=0}
        ]}}})
    Scale RGB=&Multiply_VectorByFloat(
     Vector=<1 1 1>
     Float=&FNTimeOfDayScalar(Night Scale=0.2, Morning Scale=0.5, Day Scale=0.5, Evening Scale=0.5)
    )
    ScaleA=true
    ScaleRGB=true
   ) {module=ScaleColor001}
   &ScaleSpriteSize(
    Scale Sprite Size Mode=::NewEnumerator0
    Uniform Scale Factor=&FloatFromCurve(
     FloatCurve=DataInterface() {propertyValues {Curve {keys [{arriveTangent=2.6894, interpMode=RCIM_Cubic, leaveTangent=2.6894, tangentMode=RCTM_User, time=0, value=0}, <at 1 1 RCIM_Cubic>]}}}
     Scale Curve=0.7
    )
   )
  ]
  rendererClasses [&NiagaraSpriteRendererProperties]
  renderers [{rendererClass=&NiagaraSpriteRendererProperties}]
 }
 Tail {
  emitterUpdateScript [
   &EmitterState()
   &SpawnBurst_Instantaneous(
    Spawn Probability=HlslExpression() {hlslExpression=1.0}
    Spawn Time=&Add_Float(A=$User.SecondBeat, B=1.5)
    Use Spawn Probability=false
    UseLoopCountLimit=true
   )
   /SetVariables(Emitter.Flicker:NiagaraFloat=/Niagara/DynamicInputs/Helpers/Waveform(
     Waveform Count=::NewEnumerator1
     "Waveform [1]"=::NewEnumerator0
     "[1] Frequency"=2
     "[1] Amplitude Min/Max"=<0.6 1>
     "Waveform [2] Blend"=::NewEnumerator2
     "[2] Input Position"=0
     "[2] Phase"=$Emitter.Age
     "[2] Phase Scale"=1
     "[2] Compound Frequency"=2
     "[2] Frequency"=4
     "[2] Amplitude Min/Max"=<0.6 1>
     "[2] Scale"=1
     "[2] Bias"=0
     "[2] Exponent"=1
     "[2] Pulse Duration"=0.1
     "[2] Clip Triangle Min/Max"=<0 0>
     "[2] Random Seed"=0
    )) {isSetParameters=true}
  ]
  particleSpawnScript [
   &InitializeParticle(
    Color=&MultiplyLinearColor(
     Linear Color=&FNTimeOfDayColor(Night Color=<rgba 20 6 12.0607 0.2>, Morning Color=<rgba 5 1 2.5833 3>, Day Color=<rgba 5 0 1.3634 3>, Evening Color=<rgba 5 0 1.3634 3>)
     Scale Factor=<0.02 0.02 0.02 1>
    )
    Sprite Size Mode=::NewEnumerator3
    UsePositionOffset=false
    Write Material Random=true
   )
   &DynamicMaterialParameters(
    Index 0 Param 1=0.5
    Module.Index0Param0Write=true
    Module.Index0Param1Write=true
    Module.Index0Param2Write=true
    Module.Index0Param3Write=true
   )
   /SetVariables(Particles.CameraOffset=4, Particles.CameraOffsetDark=5, Particles.DY0:Vector4f=<1 0 0 0>) {isSetParameters=true, module=SetVariables_#1}
  ]
  particleUpdateScript [
   &ParticleState(Kill Particles When Lifetime Has Elapsed=false, Let Infinitely Lived Particles Die When Emitter Deactivates=true)
   &SkeletalMeshLocation(
    Bone / Socket Index=2
    MaskSpawnGroup=false
    Mesh Sampling Type=::NewEnumerator5
    Skeletal Mesh=$User.Skeletal Mesh
    Socket Sampling Mode=::NewEnumerator5
   )
   &FN_Sprite_Attributes(
    Color=&FNTimeOfDayColor(
     Night Color=<rgba 0.4 0 0.0775 0.6>
     Morning Color=<rgba 0.4 0 0.0775 0.5>
     Day Color=<rgba 0.4 0.0612 0.2415 0.1>
     Evening Color=<rgba 0.4 0 0.0775 0.3>
    )
    Create Color Attribute=true
    Create Position Attribute=true
    Create SpriteSize Attribute=true
    Position:NiagaraPosition=&AddVectorToPosition(Position=$Particles.Position, Vector=&TransformVector(Vector=<0 0 0>, Destination Space=::World))
    SpriteSize=&Vector2DFromFloat(Value=20)
    Write to Module Namespace=true
   )
   &FN_Sprite_Attributes(
    Color=&MultiplyLinearColor(
     Linear Color=$Particles.Color
     Scale Factor=&MakeVector4_FromLinearColor(LinearColor=<rgba 3 2.94 2.94 0.03>)
    )
    Create Color Attribute=true
    Create SpriteSize Attribute=true
    SpriteSize=&Multiply_Vector2D(A=$Particles.SpriteSize, B=&Vector2DFromFloat(Value=2.5))
    Write to Module Namespace=true
   ) {module=FN_Sprite_Attributes001}
   &FN_Sprite_Attributes(
    Color=&MultiplyLinearColor(
     Linear Color=$Particles.Color
     Scale Factor=&MakeVector4_FromLinearColor(LinearColor=&MultiplyLinearColor(
       Linear Color=<rgba 3 2.94 2.94 0.07>
       Scale Factor=&MakeVector4(
        X=1
        Y=1
        Z=1
        W=&FloatFromCurve(
         FloatCurve=DataInterface() {propertyValues {Curve {keys [<at 0.4 1>,<at 1 0>]}}}
         CurveIndex=&Subtract_Float(A=$Emitter.Age, B=$User.ThirdBeat)
        )
       )
      ))
    )
    Create Color Attribute=true
    Create SpriteSize Attribute=true
    SpriteSize=&Multiply_Vector2D(
     A=&Vector2DFromCurve(Vector2Curve=DataInterface() {propertyValues {XCurve {keys [<at 0 2>,<at 1 1.5>]}, YCurve {keys [<at 0 3>,<at 0.5 1>]}}}, ScaleCurve=<1 1>)
     B=&Vector2DFromFloat(Value=100)
    )
    Write to Module Namespace=true
   ) {module=FN_Sprite_Attributes002}
   &ScaleColor(
    Scale Mode=::NewEnumerator1
    Scale RGBA=&MakeVector4_FromLinearColor(LinearColor=&FNTimeOfDayColor(Night Color=<rgba 2 2 2 10>, Morning Color=<rgba 1 1 1 1>, Day Color=<rgba 3 3 3 3>, Evening Color=<rgba 1 1 1 4>))
    ScaleRGBA=true
   ) {module=ScaleColor002}
   &ScaleColor(
    Scale Alpha=&FloatFromCurve(
     FloatCurve=DataInterface() {propertyValues {Curve {keys [<at 0.4 1>,<at 0.7 0>]}}}
     CurveIndex=&Subtract_Float(A=$Emitter.Age, B=$User.ThirdBeat)
     Scale Curve=$Emitter.Flicker
    )
    ScaleA=true
    ScaleRGB=true
   )
  ]
  rendererClasses [&NiagaraSpriteRendererProperties]
  renderers [
   {rendererClass=&NiagaraSpriteRendererProperties}
   {rendererClass=&NiagaraSpriteRendererProperties, rendererIndex=1}
   {rendererClass=&NiagaraSpriteRendererProperties, rendererIndex=2}
   {rendererClass=&NiagaraSpriteRendererProperties, rendererIndex=3}
  ]
 }
 eyes_glow_blinkfade001 {
  emitterUpdateScript [
   &EmitterState()
   &SpawnBurst_Instantaneous(Spawn Probability=HlslExpression() {hlslExpression=1.0}, Spawn Time=$User.SecondBeat, Use Spawn Probability=false, UseLoopCountLimit=true)
  ]
  particleSpawnScript [
   &InitializeParticle(
    Color=&MultiplyLinearColor(Linear Color=<rgba 1.04 0.0211 0.1665 1>, Scale Factor=<1 1 1 1>)
    Lifetime=0.5
    Sprite Size Mode=::NewEnumerator3
    Uniform Sprite Size=100
    UsePositionOffset=false
    Write Material Random=true
   )
   &DynamicMaterialParameters(
    Index 0 Param 1=0.5
    Module.Index0Param0Write=true
    Module.Index0Param1Write=true
    Module.Index0Param2Write=true
    Module.Index0Param3Write=true
   )
   /SetVariables(Particles.CameraOffset=4, Particles.CameraOffsetDark=1, Particles.DY0:Vector4f=<1 0 0 0>) {isSetParameters=true, module=SetVariables_#1}
  ]
  particleUpdateScript [
   &ParticleState()
   &ScaleSpriteSize(Uniform Curve Sprite Scale=DataInterface() {propertyValues {Curve {keys [{tangentMode=RCTM_User, time=0, value=1}, {tangentMode=RCTM_User, time=1, value=0}]}}})
   &SkeletalMeshLocation(
    Bone / Socket Index=&ReturnExecIndex()
    MaskSpawnGroup=false
    Mesh Sampling Type=::NewEnumerator5
    Skeletal Mesh=$User.Skeletal Mesh
    Socket Sampling Mode=::NewEnumerator5
   )
  ]
  rendererClasses [&NiagaraSpriteRendererProperties]
  renderers [{rendererClass=&NiagaraSpriteRendererProperties}]
 }
 eyes_glow_blinkfade003 {
  emitterUpdateScript [
   &EmitterState()
   &SpawnBurst_Instantaneous(
    Spawn Count=2
    Spawn Probability=HlslExpression() {hlslExpression=1.0}
    Spawn Time=&Subtract_Float(A=$User.ThirdBeat, B=0.1)
    Use Spawn Probability=false
    UseLoopCountLimit=true
   )
   &SpawnBurst_Instantaneous(
    Spawn Count=2
    Spawn Probability=HlslExpression() {hlslExpression=1.0}
    Spawn Time=&Subtract_Float(A=$User.ThirdBeat, B=-0.4)
    Use Spawn Probability=false
    UseLoopCountLimit=true
   ) {enabled=false, module=SpawnBurst_Instantaneous001}
  ]
  particleSpawnScript [
   &InitializeParticle(
    Color=&MultiplyLinearColor(Linear Color=<rgba 5 0 1.3634 1>, Scale Factor=<0.02 0.02 0.02 1>)
    Lifetime=0.5
    Sprite Size Mode=::NewEnumerator3
    UsePositionOffset=false
    Write Material Random=true
   )
   &DynamicMaterialParameters(
    Index 0 Param 1=0.5
    Module.Index0Param0Write=true
    Module.Index0Param1Write=true
    Module.Index0Param2Write=true
    Module.Index0Param3Write=true
   )
   /SetVariables(Particles.CameraOffset=-5, Particles.CameraOffsetDark=5, Particles.DY0:Vector4f=<1 0 0 0>) {isSetParameters=true, module=SetVariables_#1}
  ]
  particleUpdateScript [
   &ParticleState(Kill Particles When Lifetime Has Elapsed=false, Let Infinitely Lived Particles Die When Emitter Deactivates=true)
   &SkeletalMeshLocation(
    Bone / Socket Index=&ReturnExecIndex()
    MaskSpawnGroup=false
    Mesh Sampling Type=::NewEnumerator5
    Skeletal Mesh=$User.Skeletal Mesh
    Socket Sampling Mode=::NewEnumerator5
   )
   &FN_Sprite_Attributes(
    Color=&FNTimeOfDayColor(
     Night Color=<rgba 0.4 0 0.0775 0.6>
     Morning Color=<rgba 0.4 0 0.0775 0.5>
     Day Color=<rgba 0.4 0.0612 0.2415 0.1>
     Evening Color=<rgba 0.4 0 0.0775 0.3>
    )
    Create Color Attribute=true
    Create Position Attribute=true
    Create SpriteSize Attribute=true
    Position:NiagaraPosition=&AddVectorToPosition(Position=$Particles.Position, Vector=&TransformVector(Vector=<0 0 0>, Destination Space=::World))
    SpriteSize=&Vector2DFromFloat(Value=20)
    Write to Module Namespace=true
   )
   &FN_Sprite_Attributes(
    Color=&MultiplyLinearColor(
     Linear Color=$Particles.Color
     Scale Factor=&MakeVector4_FromLinearColor(LinearColor=<rgba 3 2.94 2.94 0.03>)
    )
    Create Color Attribute=true
    Create SpriteSize Attribute=true
    SpriteSize=&Multiply_Vector2D(A=$Particles.SpriteSize, B=&Vector2DFromFloat(Value=2.5))
    Write to Module Namespace=true
   ) {module=FN_Sprite_Attributes001}
   &FN_Sprite_Attributes(
    Color=&MultiplyLinearColor(
     Linear Color=$Particles.Color
     Scale Factor=&MakeVector4_FromLinearColor(LinearColor=&MultiplyLinearColor(
       Linear Color=<rgba 3 2.94 2.94 0.03>
       Scale Factor=&MakeVector4(
        X=1
        Y=1
        Z=1
        W=&FloatFromCurve(
         FloatCurve=DataInterface() {propertyValues {Curve {keys [<at 0.4 1>,<at 1 0>]}}}
         CurveIndex=&Subtract_Float(A=$Emitter.Age, B=$User.ThirdBeat)
        )
       )
      ))
    )
    Create Color Attribute=true
    Create SpriteSize Attribute=true
    SpriteSize=&Multiply_Vector2D(
     A=&Vector2DFromCurve(Vector2Curve=DataInterface() {propertyValues {XCurve {keys [<at 0 5>,<at 1 2>]}, YCurve {keys [<at 0 3>,<at 0.5 1>]}}}, ScaleCurve=<0.7 0.2>)
     B=&Vector2DFromFloat(Value=100)
    )
    Write to Module Namespace=true
   ) {module=FN_Sprite_Attributes002}
   &ScaleColor(
    Scale Mode=::NewEnumerator1
    Scale RGBA=&MakeVector4_FromLinearColor(LinearColor=&FNTimeOfDayColor(Night Color=<rgba 10 10 10 1>, Morning Color=<rgba 1 1 1 10>, Day Color=<rgba 3 3 3 3>, Evening Color=<rgba 1 1 1 4>))
    ScaleRGBA=true
   ) {module=ScaleColor002}
  ]
  rendererClasses [&NiagaraSpriteRendererProperties]
  renderers [
   {rendererClass=&NiagaraSpriteRendererProperties}
   {rendererClass=&NiagaraSpriteRendererProperties, rendererIndex=1}
   {rendererClass=&NiagaraSpriteRendererProperties, rendererIndex=2}
   {rendererClass=&NiagaraSpriteRendererProperties, rendererIndex=3}
  ]
 }
}
summary {
 systemName=NS_Emote_ReptilianOcean_SnakeEyes
 emitters [
  {emitterName=FLARE, rendererClasses [&NiagaraSpriteRendererProperties]}
  {emitterName=eyes_glow_blinkfade001, rendererClasses [&NiagaraSpriteRendererProperties]}
  {emitterName=eyes_glow_blinkfade003, rendererClasses [&NiagaraSpriteRendererProperties]}
  {emitterName=Tail, rendererClasses [&NiagaraSpriteRendererProperties]}
  {bEnabled=false, emitterName=RoundShape, rendererClasses [&NiagaraSpriteRendererProperties]}
  {
   emitterName=Default_LightAndGlowFlicker_Loop
   rendererClasses [
    &NiagaraSpriteRendererProperties
    &NiagaraLightRendererProperties
    &NiagaraMeshRendererProperties
   ]
  }
 ]
 userVariables [
  User.bIsFrontEnd {type=NiagaraBool}
  User.bIsFrontEndPreview {type=NiagaraBool}
  User.FirstBeat {type=NiagaraFloat, defaultValue=0.6, description=""}
  User.FootDustColorArray {
   type=NiagaraDataInterfaceArrayColor
   description=""
   defaultValue {
    struct=NiagaraExt_VariableValue_DataInterface
    value {
     dataInterface=&self:NiagaraDataInterfaceArrayColor_1
     dataInterfaceClass=&NiagaraDataInterfaceArrayColor
    }
   }
  }
  User.Offset {type=Vector3f, description="", defaultValue {struct=Vector3f, value=<0 0 0>}}
  User.SecondBeat {type=NiagaraFloat, defaultValue=1.8, description=""}
  User.Skeletal Mesh {
   type=NiagaraDataInterfaceSkeletalMesh
   defaultValue {
    struct=NiagaraExt_VariableValue_DataInterface
    value {dataInterface=&self:NiagaraDataInterfaceSkeletalMesh_0}
   }
  }
  User.SurfaceTypeIndex {type=NiagaraInt32}
  User.ThirdBeat {type=NiagaraFloat, defaultValue=10, description=""}
 ]
}
